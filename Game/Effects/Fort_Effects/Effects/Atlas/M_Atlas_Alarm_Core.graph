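# M_Atlas_Alarm_Core is an unlit translucent material for mesh particles, evidently the core of the Atlas alarm effect.
# A Fresnel function blends two HDR emissive color parameters, orange-red and a paler warm tone.
# The blend is multiplied by a time-of-day value, with one scalar for morning and day and another for evening and night, then by particle color.
# The product feeds a shared FX emissive adjustment function.
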Material /Game/Effects/Fort_Effects/Effects/Atlas/M_Atlas_Alarm_Core
graph {
 properties {bUsedWithMeshParticles=true, blendMode=BLEND_Translucent, pixelDepthOffsetMode=PDOM_Legacy, shadingModel=MSM_Unlit, usedShadingModels=Unlit}
 nodes [
  n1 = ScalarParameter() {defaultValue=1, parameterName=Color_Day}
  n2 = ScalarParameter() {defaultValue=1, parameterName=Color_Night}
  n3 = MaterialFunctionCall(Color_Morning=@n1, Color_Day=@n1, Color_Evening=@n2, Color_Night=@n2) {materialFunction=&Color_Time_of_Day}
  n4 = LinearInterpolate(
   A=VectorParameter() {parameterName=Emissive A, defaultValue=<rgba 10 0.3333 0 0>}.RGB
   B=VectorParameter() {parameterName=Emissive B, defaultValue=<rgba 5 4.0685 2.1189 0>}.RGB
   Alpha=MaterialFunctionCall(
    Brightness=ScalarParameter() {defaultValue=1.5, parameterName=Fresnel Brightness}
    Exponent=ScalarParameter() {defaultValue=10, parameterName=Fresnel Exponent}
   ) {materialFunction=&MF_Fresnel_FX}.Fresnel
  )
  MaterialFunctionCall(Emissive=Multiply(A=Multiply(A=@n3.Result, B=@n4), B=ParticleColor().RGB)) {materialFunction=&MF_FX_EmissiveAdjustments}
 ]
}
