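# Translucent, two-sided unlit effect material, apparently for the trail on the Atlas sphere, usable on particle, skeletal and deformed meshes.
# Time-driven rotation about a slowly wobbling axis through the actor position, plus a sine Z bob, produces a vertex offset.
# The same rotation is applied to a normal vector converted back to tangent space.
# Constant blue emissive runs through MF_FX_EmissiveAdjustments; opacity combines a ramp texture, particle alpha and Fresnel.
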Material /Game/Effects/Fort_Effects/Effects/Atlas/M_Atlas_Sphere_Trail
graph {
 properties {
  bUsedWithMeshDeformer=true
  bUsedWithMeshParticles=true
  bUsedWithSkeletalMesh=true
  blendMode=BLEND_Translucent
  pixelDepthOffsetMode=PDOM_Legacy
  shadingModel=MSM_Unlit
  twoSided=true
  usedShadingModels=Unlit
 }
 nodes [
  n1 = Sine(None=Multiply(B=Time()) {constA=0.1}) {period=3}
  n2 = Normalize(VectorInput=LinearInterpolate(A=Constant3Vector() {constant=<rgba 0 0 1 0>}, B=Constant3Vector() {constant=<rgba 1 0 0 0>}, Alpha=ConstantBiasScale(None=@n1) {scale=0.3}))
  n3 = Multiply(A=Time()) {constB=0.15}
  n4 = ActorPositionWS()
  n5 = Time()
  Add(
   A=RotateAboutAxis(NormalizedRotationAxis=@n2, RotationAngle=@n3, PivotPoint=@n4, Position=WorldPosition().XYZ)
   B=Multiply(A=Sine(None=@n5) {desc=0, period=3.5}, B=Constant3Vector() {constant=<rgba 0 0 3 0>})
  )
  MaterialFunctionCall(Emissive=Constant3Vector() {constant=<rgba 1 2 3 0>}) {materialFunction=&MF_FX_EmissiveAdjustments}
  Multiply(
   A=TextureSample() {texture=/Game/Global/Textures/Default/Masks_and_Gradients/T_Water_Top_Ramp}.G
   B=Multiply(A=ParticleColor().A, B=Fresnel())
  )
  n6 = Constant3Vector() {constant=<rgba 0 0 1 0>}
  Transform(None=RotateAboutAxis(NormalizedRotationAxis=@n2, RotationAngle=@n3, PivotPoint=@n4, Position=Transform(None=@n6))) {transformSourceType=TRANSFORMSOURCE_World, transformType=TRANSFORM_Tangent}
 ]
}
