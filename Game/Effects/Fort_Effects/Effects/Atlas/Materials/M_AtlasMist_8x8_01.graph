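# Translucent, volumetric per-vertex directional lit sprite particle material that plays an 8x8 falling-mist flipbook.
# The G channel, shaped by Color Contrast and tinted by particle color, feeds an emissive adjustment function with Emissive Boost.
# The B channel builds opacity from a contrast term plus a high-power edge term, both scaled by particle alpha, then depth-faded and reduced by dynamic Erode.
# A spherical normal map tiles 8x8, matching the flipbook grid.
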Material /Game/Effects/Fort_Effects/Effects/Atlas/Materials/M_AtlasMist_8x8_01
graph {
 properties {
  bUsedWithParticleSprites=true
  blendMode=BLEND_Translucent
  pixelDepthOffsetMode=PDOM_Legacy
  translucencyLightingMode=TLM_VolumetricPerVertexDirectional
  usedShadingModels=Default Lit
 }
 nodes [
  n1 = DynamicParameter() {defaultValue=<rgba 1 0.1 1 0>, paramNames [Alpha Contrast, Emissive Boost, Color Contrast, Erode]}
  n2 = ParticleColor()
  n3 = ParticleSubUV!(UVs, Tex, Apply View MipBias) {samplerType=SAMPLERTYPE_LinearColor, texture=/Game/Effects/Fort_Effects/Textures/Smoke/T_FallingMist_8x8_01}
  n4 = DynamicParameter() {defaultValue=<rgba 1 0.1 1 0>, paramNames [Alpha Contrast, Emissive Boost, Color Contrast, Erode]}
  n5 = Power(Base=@n3.G, Exp=@n4.Color Contrast)
  n6 = ParticleColor()
  MaterialFunctionCall(Emissive=Multiply(A=Multiply(A=@n1.Emissive Boost, B=@n2.RGB), B=Multiply(A=@n5, B=@n6.RGB))) {materialFunction=&MF_FX_EmissiveAdjustments}
  n7 = ScalarParameter() {defaultValue=50, parameterName=Edge Power}
  n8 = Multiply(A=Multiply(A=Power(Base=@n3.B, Exp=@n7) {constExponent=20}, B=ScalarParameter() {defaultValue=1.5, parameterName=Edge Brightness}), B=@n6.A)
  n9 = Multiply(A=Power(Base=@n3.B, Exp=@n4.Alpha Contrast), B=@n6.A)
  Subtract(
   A=DepthFade(Opacity=Add(A=@n8, B=@n9), FadeDistance=ScalarParameter() {defaultValue=70, parameterName=Depth Fade Distance})
   B=DynamicParameter() {defaultValue=<rgba 1 0.1 1 0>, paramNames [Alpha Contrast, Emissive Boost, Color Contrast, Erode]}.Erode
  )
  TextureSample(UVs=TextureCoordinate() {uTiling=8, vTiling=8}) {samplerType=SAMPLERTYPE_Normal, texture=/Game/Gadgets/Assets/Teleporter/Textures/T_Spherical_Normal_02}
 ]
}
