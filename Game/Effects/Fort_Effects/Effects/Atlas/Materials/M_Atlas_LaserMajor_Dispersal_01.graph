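# Unlit, translucent, two-sided mesh-particle material for the Atlas laser dispersal effect, blue by default.
# Three panning texture samples on power-warped V coordinates form one noise, with the two beam strands scaled by a dynamic parameter.
# That noise blends the colors, feeds opacity and pushes vertices along normals, faded by vertex alpha and V.
# A V gradient drives emissive intensity and adds opacity, while the dynamic Mask offsets a V-based opacity wipe.
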Material /Game/Effects/Fort_Effects/Effects/Atlas/Materials/M_Atlas_LaserMajor_Dispersal_01
graph {
 properties {
  bUsedWithMeshParticles=true
  blendMode=BLEND_Translucent
  ditherOpacityMask=true
  opacityMaskClipValue=0.25
  pixelDepthOffsetMode=PDOM_Legacy
  shadingModel=MSM_Unlit
  translucencyLightingMode=TLM_VolumetricDirectional
  twoSided=true
  usedShadingModels=Unlit
 }
 nodes [
  n1 = MaterialFunctionCall(Float2=TextureCoordinate()) {materialFunction=&BreakOutFloat2Components}
  n2 = AppendVector(A=@n1.R, B=Power(Base=@n1.G, Exp=ScalarParameter() {defaultValue=0.4, parameterName=V Power}))
  n3 = ScalarParameter() {defaultValue=1, parameterName=U1}
  n4 = ScalarParameter() {defaultValue=0.2991, parameterName=V1}
  n5 = Panner(
   Coordinate=Multiply(A=@n2, B=AppendVector(A=@n3, B=@n4))
   Time=Multiply(A=Time(), B=ScalarParameter() {defaultValue=0.0467, parameterName=Speed 1})
  ) {bFractionalPart=true, speedX=0.178, speedY=1}
  n6 = Multiply(A=@n2, B=AppendVector(A=ScalarParameter() {defaultValue=1, parameterName=U2}, B=ScalarParameter() {defaultValue=0.3, parameterName=V2}))
  n7 = Multiply(A=Time(), B=ScalarParameter() {defaultValue=0.076, parameterName=Speed 2})
  n8 = ScalarParameter() {defaultValue=1, parameterName=U3}
  n9 = ScalarParameter() {defaultValue=0.25, parameterName=V3}
  n10 = Add(A=Multiply(A=@n2, B=AppendVector(A=@n8, B=@n9))) {constB=0.5}
  n11 = Multiply(A=Time(), B=ScalarParameter() {defaultValue=0.0674, parameterName=Speed 3})
  n12 = Add(
   A=TextureSample(UVs=Panner(Coordinate=@n6, Time=@n7) {bFractionalPart=true, speedX=-0.223, speedY=1}) {samplerType=SAMPLERTYPE_LinearColor, texture=/Game/Effects/Fort_Effects/Textures/Tile/T_Raygun_Beam_02}.G
   B=TextureSample(UVs=Panner(Coordinate=@n10, Time=@n11) {bFractionalPart=true, speedX=0.347, speedY=1}) {samplerType=SAMPLERTYPE_LinearColor, texture=&T_Raygun_Beam_02}.R
  )
  n13 = DynamicParameter() {paramNames [Mask, Strands Opacity, Param3, Param4]}
  n14 = Divide(A=Add(
    A=TextureSample(UVs=@n5) {texture=&2ndInfectedMapLinear}.RGB
    B=Multiply(A=@n12, B=@n13.Strands Opacity)
   )) {constB=3}
  n15 = Multiply(A=@n14, B=VertexNormalWS())
  n16 = VectorParameter() {parameterName=WPO Strength, defaultValue=<rgba 2 2 2 0>}
  n17 = TextureCoordinate()
  LinearInterpolate(B=LinearInterpolate(B=Multiply(A=@n15, B=@n16.RGB), Alpha=VertexColor().A), Alpha=OneMinus(None=ComponentMask(None=@n17) {g=true}))
  n18 = VectorParameter() {parameterName=Base Color B, defaultValue=<rgba 0.08 0.2858 1 1>}
  n19 = LinearInterpolate(A=VectorParameter() {parameterName=Base Color A, defaultValue=<rgba 0.0014 0.0089 0.035 1>}.RGB, B=@n18.RGB, Alpha=@n14)
  n20 = TextureCoordinate()
  n21 = ScalarParameter() {defaultValue=0.1, parameterName=Emissive Gradient Bias}
  n22 = Power(Base=ComponentMask(None=Add(A=@n20, B=@n21)) {g=true}, Exp=ScalarParameter() {defaultValue=2.5, parameterName=Emissive Gradient Exponent})
  n23 = ScalarParameter() {defaultValue=0.1, parameterName=Emissive Top}
  n24 = ScalarParameter() {defaultValue=10, parameterName=Emissive Bottom}
  MaterialFunctionCall(Emissive=Multiply(A=LinearInterpolate(A=@n19, B=@n18.RGB, Alpha=@n22), B=LinearInterpolate(A=@n23, B=@n24, Alpha=@n22))) {materialFunction=&MF_FX_EmissiveAdjustments}
  n25 = ScalarParameter() {defaultValue=0.3, parameterName=Opacity BG Boost}
  n26 = Add(A=@n22, B=Multiply(A=VertexColor().A, B=Add(A=@n25, B=@n14)))
  n27 = ComponentMask(None=TextureCoordinate()) {g=true}
  n28 = DynamicParameter() {paramNames [Mask, Strands Opacity, Param3, Param4]}
  Multiply(
   A=Power(Base=Clamp(None=@n26), Exp=ScalarParameter() {defaultValue=1, parameterName=Opacity Exponent})
   B=Clamp(None=Add(A=@n27, B=@n28.Mask))
  )
 ]
}
