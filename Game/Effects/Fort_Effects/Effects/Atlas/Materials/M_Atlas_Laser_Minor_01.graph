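# Unlit material for mesh particles on the minor laser effect of Atlas, previewed on its laser mesh.
# A Fresnel term plus a tiled, panning muzzle texture blends outer and inner blue colors, then passes through emissive adjustments.
# Two summed sines of time drive a vertex-normal offset for a pulsing width, with a dynamic particle parameter biasing both phase and amount.
# Several vector experiments, including a view-dependent brightness blend, feed nothing visible.
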Material /Game/Effects/Fort_Effects/Effects/Atlas/Materials/M_Atlas_Laser_Minor_01
graph {
 properties {
  bUsedWithMeshParticles=true
  pixelDepthOffsetMode=PDOM_Legacy
  previewMesh=/Game/Effects/Fort_Effects/Effects/Atlas/Mesh/Sm_Atlas_LaserMinor_01
  shadingModel=MSM_Unlit
  usedShadingModels=Unlit
 }
 nodes [
  n1 = Constant3Vector() {constant=<rgba 1 0 0 0>}
  CrossProduct(
   A=MaterialFunctionCall() {materialFunction=&CameraDirectionVector}.Result
   B=Normalize(VectorInput=Transform(None=@n1) {transformSourceType=TRANSFORMSOURCE_Local})
  )
  DotProduct(B=PixelNormalWS())
  n2 = ScalarParameter() {defaultValue=1.5, parameterName=Fresnel Brightness A}
  n3 = Abs(None=CameraVectorWS())
  n4 = Constant3Vector() {constant=<rgba 1 0 0 0>}
  LinearInterpolate(
   A=@n2
   B=ScalarParameter() {defaultValue=5, parameterName=Fresnel Brightness B}
   Alpha=Power(Base=DotProduct(A=@n3, B=@n4)) {constExponent=5}
  )
  n5 = MaterialFunctionCall(Brightness=@n2, Exponent=ScalarParameter() {defaultValue=8, parameterName=Fresnel Exponent}) {materialFunction=&MF_Fresnel_FX}
  n6 = TextureCoordinate()
  n7 = AppendVector(A=ScalarParameter() {defaultValue=1, parameterName=U Tile}, B=ScalarParameter() {defaultValue=0.27, parameterName=V Tile})
  n8 = TextureSample(UVs=Panner(Coordinate=Multiply(A=@n6, B=@n7)) {bFractionalPart=true, speedX=-0.87, speedY=-2.77}) {samplerType=SAMPLERTYPE_LinearColor, texture=/Game/Effects/Fort_Effects/Textures/Tile/T_Raygun_Muzzle_02}
  MaterialFunctionCall(Emissive=LinearInterpolate(
    A=VectorParameter() {parameterName=Color Outer, defaultValue=<rgba 0 0.1 0.3 0>}.RGB
    B=VectorParameter() {parameterName=Color Inner, defaultValue=<rgba 1 3.041 50 0>}.RGB
    Alpha=Add(A=@n5.Fresnel, B=@n8.R)
   )) {materialFunction=&MF_FX_EmissiveAdjustments}
  n9 = ScalarParameter() {defaultValue=-1.45, parameterName=WPO Min}
  n10 = ScalarParameter() {defaultValue=-0.5, parameterName=WPO Max}
  n11 = Multiply(
   A=Add(A=DynamicParameter() {defaultValue=<rgba 0 1 1 1>, paramNames [WPO Bias, Param2, Param3, Param4]}.WPO Bias, B=Time())
   B=ScalarParameter() {defaultValue=1.5, parameterName=Speed}
  )
  n12 = Add(A=Sine(None=@n11) {period=0.104}, B=Sine(None=@n11) {period=0.178})
  Multiply(
   A=VertexNormalWS()
   B=Add(
    A=LinearInterpolate(A=@n9, B=@n10, Alpha=@n12)
    B=DynamicParameter() {defaultValue=<rgba 0 1 1 1>, paramNames [WPO Bias, Param2, Param3, Param4]}.WPO Bias
   )
  )
  n13 = VertexNormalWS()
  n14 = ObjectBounds()
  Transform(None=Normalize(VectorInput=Divide(A=@n13, B=@n14))) {transformSourceType=TRANSFORMSOURCE_World, transformType=TRANSFORM_Tangent}
  Transform(None=PixelNormalWS()) {transformSourceType=TRANSFORMSOURCE_World, transformType=TRANSFORM_View}
 ]
}
