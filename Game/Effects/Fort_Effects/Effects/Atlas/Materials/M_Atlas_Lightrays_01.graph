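# Two-sided unlit translucent material for Atlas light ray effects on Niagara sprites and mesh particles.
# A sharpened V-coordinate gradient adds to a depth-faded Fresnel term, then is scaled by a fader and near-camera fade.
# That mask blends two colors or tints particle color before emissive adjustments.
# A separate alpha branch optionally applies particle alpha and a world-space sphere mask around the offset actor position, despite the mask's "Emissive" group name.
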Material /Game/Effects/Fort_Effects/Effects/Atlas/Materials/M_Atlas_Lightrays_01
graph {
 properties {
  bUsedWithNiagaraMeshParticles=true
  bUsedWithNiagaraSprites=true
  blendMode=BLEND_Translucent
  pixelDepthOffsetMode=PDOM_Legacy
  shadingModel=MSM_Unlit
  twoSided=true
  usedShadingModels=Unlit
 }
 nodes [
  n1 = ScalarParameter() {group="4.2) Emissive Sphere Mask", parameterName=Sphere Mask Low Value, sortPriority=1}
  n2 = ScalarParameter() {defaultValue=1, group="4.2) Emissive Sphere Mask", parameterName=Sphere Mask High Value, sortPriority=2}
  n3 = WorldPosition()
  n4 = VectorParameter() {group="4.2) Emissive Sphere Mask", parameterName=Location Offset, sortPriority=5, defaultValue=<rgba 0 0 0 1>}
  n5 = Add(A=ActorPositionWS(), B=@n4.RGB)
  n6 = ScalarParameter() {defaultValue=50, group="4.2) Emissive Sphere Mask", parameterName=Sphere Mask Radius, sortPriority=3}
  n7 = ScalarParameter() {group="4.2) Emissive Sphere Mask", parameterName=Sphere Mask Hardness, sortPriority=4}
  n8 = Clamp(None=Power(
    Base=SphereMask(A=@n3.XYZ, B=@n5, Radius=@n6, Hardness=@n7)
    Exp=ScalarParameter() {group="4.2) Emissive Sphere Mask", parameterName=Power, sortPriority=5}
   ) {constExponent=1})
  n9 = ComponentMask(None=TextureCoordinate()) {g=true}
  n10 = Power(Base=@n9, Exp=ScalarParameter() {defaultValue=12, parameterName=Gradient Exponent}) {constExponent=8}
  n11 = MaterialFunctionCall(
   Brightness=ScalarParameter() {defaultValue=1.2, parameterName=Frensel Brightness}
   Exponent=ScalarParameter() {defaultValue=2, parameterName=Frensel Exponent}
  ) {materialFunction=&MF_Fresnel_FX}
  n12 = DepthFade(
   Opacity=Multiply(A=Multiply(A=@n9, B=@n11.Fresnel), B=ScalarParameter() {defaultValue=0.2, parameterName=OpacityBase})
   FadeDistance=ScalarParameter() {defaultValue=50, parameterName=Depth Fade Distance}
  )
  n13 = Multiply(
   A=Multiply(A=Add(A=@n10, B=@n12), B=ScalarParameter() {defaultValue=1, parameterName=Fader})
   B=MaterialFunctionCall(Distance=ScalarParameter() {defaultValue=50, parameterName=Near Camera Fade}) {materialFunction=&Near_Camera_Fade}.Result
  )
  n14 = LinearInterpolate(
   A=@n13
   B=Multiply(A=@n13, B=LinearInterpolate(A=@n1, B=@n2, Alpha=@n8))
   Alpha=ScalarParameter() {group="4.2) Emissive Sphere Mask", parameterName="Use Sphere Mask (World)", sortPriority=-1}
  )
  n15 = ParticleColor()
  LinearInterpolate(A=@n14, B=Multiply(A=@n15.A, B=@n14), Alpha=ScalarParameter() {parameterName=Use Particle Alpha})
  n16 = VectorParameter() {parameterName=Color A, defaultValue=<rgba 0.0275 0.1063 0.5 1>}
  n17 = VectorParameter() {parameterName=Color B, defaultValue=<rgba 0.55 2.125 10 1>}
  MaterialFunctionCall(Emissive=LinearInterpolate(
    A=LinearInterpolate(A=@n16.RGB, B=@n17.RGB, Alpha=@n13)
    B=Multiply(A=@n15.RGB, B=@n13)
    Alpha=ScalarParameter() {parameterName=Use Particle Color}
   )) {materialFunction=&MF_FX_EmissiveAdjustments}
 ]
}
