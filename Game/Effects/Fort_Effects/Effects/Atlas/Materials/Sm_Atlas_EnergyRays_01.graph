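# Translucent, two-sided Atlas energy-ray material for meshes and Niagara mesh particles, with animated plasma in tinted emissive colors.
# Two panning plasma noises are summed and contrasted to blend the colors and, with Fresnel and a V-gradient, shape depth-faded opacity.
# A sharpened gradient adds bright bands, optionally scrolled as moving lines.
# Panning noise pushes vertices along normals for switchable wobble, suppressed toward the gradient's bright end.
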Material /Game/Effects/Fort_Effects/Effects/Atlas/Materials/Sm_Atlas_EnergyRays_01
graph {
 properties {
  bEnableResponsiveAA=true
  bUsedWithNiagaraMeshParticles=true
  blendMode=BLEND_Translucent
  pixelDepthOffsetMode=PDOM_Legacy
  translucencyLightingMode=TLM_VolumetricPerVertexNonDirectional
  twoSided=true
  usedShadingModels=Default Lit
 }
 nodes [
  n1 = TextureCoordinate() {coordinateIndex=1}
  n2 = AppendVector(A=ScalarParameter() {defaultValue=0.5, parameterName=U WPO}, B=ScalarParameter() {defaultValue=0.3269, parameterName=V WPO})
  n3 = TextureSample(UVs=Panner(Coordinate=Multiply(A=@n1, B=@n2)) {speedX=0.33, speedY=0.2643}) {texture=&T_NoisePack_01}
  ConstantBiasScale(None=@n3.R) {bias=-0.5, scale=2}
  n4 = VectorParameter() {parameterName=Color A, defaultValue=<rgba 0.15 0.9583 5 1>}
  n5 = VectorParameter() {parameterName=Color B, defaultValue=<rgba 0.12 1.1 6 1>}
  n6 = ScalarParameter() {defaultValue=1, parameterName=U1}
  n7 = ScalarParameter() {defaultValue=0.327, parameterName=V1}
  n8 = Panner(Coordinate=Multiply(A=TextureCoordinate(), B=AppendVector(A=@n6, B=@n7))) {bFractionalPart=true, speedX=0.09, speedY=0.1491}
  n9 = ScalarParameter() {defaultValue=1, parameterName=U2}
  n10 = ScalarParameter() {defaultValue=0.1923, parameterName=V2}
  n11 = Panner(Coordinate=Multiply(A=TextureCoordinate(), B=AppendVector(A=@n9, B=@n10))) {bFractionalPart=true, speedX=-0.067, speedY=0.1428}
  n12 = Power(
   Base=Add(
    A=TextureSample(UVs=@n8) {texture=/Game/Effects/Fort_Effects/Textures/Plasma/T_PlasmaNoise_A_01}.R
    B=TextureSample(UVs=@n11) {texture=&T_PlasmaNoise_A_01}.B
   )
   Exp=ScalarParameter() {defaultValue=3, parameterName=Constrast}
  )
  n13 = Clamp(None=@n12)
  MaterialFunctionCall(Emissive=Multiply(A=LinearInterpolate(A=@n4.RGB, B=@n5.RGB, Alpha=@n13), B=ScalarParameter() {defaultValue=0.5, parameterName=Emissive Boost})) {materialFunction=&MF_FX_EmissiveAdjustments}
  n14 = TextureCoordinate() {vTiling=4}
  n15 = ScalarParameter() {defaultValue=1, parameterName=Line_Numbers}
  n16 = Time()
  n17 = MaterialFunctionCall(G=ScalarParameter() {defaultValue=40, parameterName=Line_VerticalSpeed}) {materialFunction=&AppendMany}
  n18 = StaticSwitchParameter(True=Add(A=Multiply(A=@n14, B=@n15), B=Multiply(A=@n16, B=@n17.RG)), False=@n14) {parameterName=UseLinesPanner}
  n19 = OneMinus(None=ComponentMask(None=Frac(None=@n18)) {g=true})
  n20 = Power(Base=@n19, Exp=ScalarParameter() {defaultValue=20, parameterName=Base Add Exponent}) {constExponent=16}
  Multiply(A=@n20, B=ScalarParameter() {parameterName=Base Add Scale})
  n21 = Multiply(A=@n3.R, B=VertexNormalWS())
  n22 = VectorParameter() {parameterName=WPO Intensity, defaultValue=<rgba 8 8 8 0>}
  StaticSwitchParameter(True=LinearInterpolate(A=Multiply(A=@n21, B=@n22.RGB), Alpha=Power(Base=@n19) {constExponent=4}) {constB=0}, False=Constant()) {defaultValue=true, parameterName=UseWPO?}
  n23 = ScalarParameter() {defaultValue=1, parameterName=Fresnel Brightness}
  n24 = ScalarParameter() {defaultValue=4, parameterName=Fresnel Exponent}
  n25 = Multiply(
   A=Multiply(
    A=MaterialFunctionCall(Brightness=@n23, Exponent=@n24) {materialFunction=&MF_Fresnel_FX}.Fresnel
    B=@n19
   )
   B=@n12
  )
  n26 = ScalarParameter() {defaultValue=25, parameterName=Depth Fade Distance}
  n27 = Multiply(A=Add(A=DepthFade(Opacity=@n25, FadeDistance=@n26), B=@n20), B=ScalarParameter() {defaultValue=0.5, parameterName=Fader})
  n28 = DepthFade(Opacity=Constant() {r=1}, FadeDistance=ScalarParameter() {defaultValue=15, parameterName=DepthFade-Distance})
  StaticSwitchParameter(True=ShadingPathSwitch(Default=Multiply(A=@n27, B=@n28)) {desc=Migrated from FeatureLevelSwitch}, False=@n27) {parameterName=UseDepthFade}
 ]
}
