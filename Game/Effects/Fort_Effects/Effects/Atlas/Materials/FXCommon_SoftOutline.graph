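# VFX material function that outputs a soft screen-space outline, with optional inner glow, around geometry within MaxBlurDepth behind the rendered pixel.
# Custom HLSL radially samples depth with TAA-jittered noise, using thickness scaled by camera distance.
# Static switches choose stencil-filtered, custom-depth or scene-depth variants; forward and mobile always use scene depth.
# Subtracting the center mask from the blur forms the outline; that mask is also output.
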MaterialFunction /Game/Effects/Fort_Effects/Effects/Atlas/Materials/FXCommon_SoftOutline
graph {
 properties {bExposeToLibrary=true, previewBlendMode=BLEND_Translucent, libraryCategoriesText [VFX]}
 nodes [
  MaterialExpressionFunctionInput_13 = FunctionInput() {bUsePreviewValueAsDefault=true, inputName=GlowMeshStencilIndex, inputType=FunctionInput_Scalar, sortPriority=-1000, previewValue=<40 0 0 1>}
  n1 = FunctionInput() {bUsePreviewValueAsDefault=true, inputName=OutlineThickness, inputType=FunctionInput_Scalar, sortPriority=-10, previewValue=<1.5 0 0 1>}
  n2 = SceneTexelSize()
  n3 = Subtract(
   A=FunctionInput(Preview=ParticlePositionWS()) {bUsePreviewValueAsDefault=true, inputName=MeshPosition, sortPriority=110}
   B=CameraPositionWS()
  )
  n4 = FunctionInput() {bUsePreviewValueAsDefault=true, inputName=------LessUse------, inputType=FunctionInput_Scalar, sortPriority=100}
  n5 = Divide(
   A=Divide(A=@n1, B=Multiply(A=@n2) {constB=1440})
   B=Clamp(None=MaterialFunctionCall(Vector 3=@n3, Vector 2=@n4) {materialFunction=/Engine/Functions/Engine_MaterialFunctions02/Utility/VectorLength}.V3 Length) {clampMode=CMODE_ClampMin, minDefault=30}
  )
  MaterialExpressionReroute_2 = Reroute(None=@n5)
  n6 = FunctionInput() {
   bUsePreviewValueAsDefault=true
   description=Number of steps along distance
   inputName=Distance Steps
   inputType=FunctionInput_Scalar
   sortPriority=110
   previewValue=<6 0 0 1>
  }
  MaterialExpressionReroute_3 = Reroute(None=@n6)
  n7 = FunctionInput() {
   bUsePreviewValueAsDefault=true
   description=Number of radial steps to perform
   inputName=Radial Steps
   inputType=FunctionInput_Scalar
   sortPriority=120
   previewValue=<4 0 0 1>
  }
  MaterialExpressionReroute_4 = Reroute(None=@n7)
  n8 = Constant() {r=1}
  MaterialExpressionReroute_5 = Reroute(None=@n8)
  n9 = Constant() {r=1}
  MaterialExpressionReroute_6 = Reroute(None=@n9)
  n10 = TextureObject() {samplerType=SAMPLERTYPE_LinearGrayscale, texture=/Engine/EngineMaterials/Good64x64TilingNoiseHighFreq}
  MaterialExpressionReroute_7 = Reroute(None=@n10)
  n11 = Constant() {r=1}
  MaterialExpressionReroute_8 = Reroute(None=@n11)
  n12 = Constant() {r=0.62}
  MaterialExpressionReroute_9 = Reroute(None=@n12)
  n13 = FunctionInput() {bUsePreviewValueAsDefault=true, inputName=MaxBlurDepth, inputType=FunctionInput_Scalar, sortPriority=10, previewValue=<20 0 0 1>}
  MaterialExpressionReroute_10 = Reroute(None=@n13)
  n14 = SceneTexture() {sceneTextureId=PPI_CustomDepth}
  n15 = SceneDepth()
  n16 = StaticBool()
  n17 = FunctionInput(Preview=@n16) {bUsePreviewValueAsDefault=true, inputName=UseCustomDepth, inputType=FunctionInput_StaticBool, sortPriority=-1000, previewValue=<20 0 0 1>}
  n18 = ShadingPathSwitch(Default=@n17, Deferred=@n17, Forward=@n16, Mobile=@n16)
  n19 = StaticSwitch(True=Multiply(A=ComponentMask(None=@n14.Color) {r=true}, B=@n15.R), False=@n15.R, Value=@n18)
  n20 = Multiply(A=@n19, B=SceneTexture() {sceneTextureId=PPI_CustomStencil}.Color)
  n21 = FunctionInput(Preview=@n16) {bUsePreviewValueAsDefault=true, inputName=UseCustomStencil, inputType=FunctionInput_StaticBool, sortPriority=-1000, previewValue=<20 0 0 1>}
  n22 = ShadingPathSwitch(Default=@n21, Deferred=@n21, Forward=@n16, Mobile=@n16)
  n23 = Multiply(A=Clamp(None=StaticSwitch(True=@n20, False=@n19, Value=@n22)) {minDefault=1}, B=PixelDepth().R)
  MaterialExpressionReroute_11 = Reroute(None=@n23)
  n24 = FunctionInput() {bUsePreviewValueAsDefault=true, inputName=PixelDepthOffset, inputType=FunctionInput_Scalar, sortPriority=10}
  MaterialExpressionReroute_12 = Reroute(None=@n24)
  n25 = Custom!(
   Distance=@MaterialExpressionReroute_2
   DistanceSteps=@MaterialExpressionReroute_3
   RadialSteps=@MaterialExpressionReroute_4
   TempAARotation=@MaterialExpressionReroute_5
   DistanceMask=@MaterialExpressionReroute_6
   Tex=@MaterialExpressionReroute_7
   TempAADistance=@MaterialExpressionReroute_8
   RadialOffset=@MaterialExpressionReroute_9
   DivideF=@MaterialExpressionReroute_10
   PDepth=@MaterialExpressionReroute_11
   PixelDepthOffset=@MaterialExpressionReroute_12
   StencilIndexIn=@MaterialExpressionFunctionInput_13
  ) {
   code="float CurColor = 0;\r\nfloat2 BaseUV = MaterialFloat2(ScreenAlignedPosition(Parameters.ScreenPosition).xy);\r\nfloat2 NewUV = BaseUV;\r\nfloat StepSize = Distance / (int) DistanceSteps;\r\nfloat CurDistance = 0;\r\nfloat2 CurOffset = 0;\r\nfloat TwoPi = 6.283185;\r\nfloat Substep = 0;\r\nfloat2 ScenePixels=View.BufferSizeAndInvSize.xy*BaseUV;\r\nScenePixels+=View.TemporalAAParams.r;\r\nfloat2 RandomSamp = ((uint)(ScenePixels.x) + 2 * (uint)(ScenePixels.y)) % 5;\r\nRandomSamp+=Texture2DSample(Tex,TexSampler,ScenePixels);\r\nRandomSamp/=5;\r\nRandomSamp-=0.5;\r\nTempAARotation*=RandomSamp;\r\nTempAADistance*=StepSize*RandomSamp;\r\n\r\n\r\nint i=0;\r\nif (DistanceSteps<1)\r\n{\r\n    Mask = 0;\r\n    return 0;\r\n}\r\nelse\r\n{\r\n    //CurDistance += 0.5*StepSize;\r\n    while ( i < (int) DistanceSteps)\r\n    {\r\n        //CurDistance+=StepSize;\r\n        for (int j = 0; j < (int) RadialSteps; j++)\r\n        {\r\n            CurOffset.x = cos(TwoPi*((TempAARotation+Substep) / RadialSteps));\r\n            CurOffset.y = sin(TwoPi*((TempAARotation+Substep) / RadialSteps));\r\n            CurOffset *=DistanceMask;\r\n            NewUV.x = BaseUV.x + (CurOffset.x * (CurDistance+(RandomSamp*TempAADistance)));\r\n            NewUV.y = BaseUV.y + (CurOffset.y * (CurDistance+(RandomSamp*TempAADistance)));\r\n            float tmpdepth = CalcSceneDepth(NewUV)-PDepth+PixelDepthOffset;\r\n            float tmpstencil = SceneTextureLookupCommon(Parameters, NewUV, 25, false).r;\r\n            float validmask = 1-floor(saturate(abs(tmpstencil-StencilIndexIn)+0.01));\r\n            CurColor += (min(1-saturate(max(0,(tmpdepth))/(DivideF+PixelDepthOffset))-0.99,0)+0.99)*saturate(sign(tmpdepth))*validmask;\r\n            //CurDistance+=(StepSize+(TempAADistance))/RadialSteps;\r\n            Substep++;\r\n        }\r\n        CurDistance+=StepSize;\r\n        Substep+=RadialOffset;\r\n        i++;\r\n    }\r\n    float tmpdepth2 = CalcSceneDepth(BaseUV)-PDepth+PixelDepthOffset;\r\n    float tmpstencil2 = SceneTextureLookupCommon(Parameters, BaseUV, 25, false).r;\r\n    float validmask2 = 1-floor(saturate(abs(tmpstencil2-StencilIndexIn)+0.01));\r\n    float mask = (min(1-saturate(max(0,(tmpdepth2))/(DivideF+PixelDepthOffset))-0.99,0)+0.99)*saturate(sign(tmpdepth2))*validmask2;\r\n    Mask = mask;\r\n    CurColor = CurColor / ((int)DistanceSteps*(int)RadialSteps);\r\n    return CurColor;\r\n}\r\n"
   desc="\n*** Original source before expression upgrade ***\nfloat3 CurColor;\r\nfloat2 BaseUV = MaterialFloat2(ScreenAlignedPosition(Parameters.ScreenPosition).xy);\r\nfloat2 NewUV = BaseUV;\r\nfloat StepSize = Distance / (int) DistanceSteps;\r\nfloat CurDistance=0;\r\nfloat2 CurOffset;\r\nfloat TwoPi = 6.283185;\r\nfloat Substep;\r\nfloat2 ScenePixels=View.RenderTargetSize*BaseUV;\r\nScenePixels+=View.TemporalAAParams.r;\r\nfloat2 RandomSamp = ((uint)(ScenePixels.x) + 2 * (uint)(ScenePixels.y)) % 5;\r\nRandomSamp+=Texture2DSample(Tex,TexSampler,ScenePixels);\r\nRandomSamp/=5;\r\nRandomSamp-=0.5;\r\nTempAARotation*=RandomSamp;\r\nTempAADistance*=StepSize*RandomSamp;\r\n\r\n\r\nint i=0;\r\nif (DistanceSteps<1)\r\n{\r\nreturn DecodeSceneColorForMaterialNode(NewUV);\r\n}\r\nelse\r\n{\r\n//CurDistance += 0.5*StepSize;\r\nwhile ( i < (int) DistanceSteps)\r\n{\r\n\r\n//CurDistance+=StepSize;\r\nfor (int j = 0; j < (int) RadialSteps; j++)\r\n{\r\nCurOffset.x = cos(TwoPi*((TempAARotation+Substep) / RadialSteps));\r\nCurOffset.y = sin(TwoPi*((TempAARotation+Substep) / RadialSteps));\r\nCurOffset *=DistanceMask;\r\nNewUV.x = BaseUV.x + (CurOffset.x * (CurDistance+(RandomSamp*TempAADistance)));\r\nNewUV.y = BaseUV.y + (CurOffset.y * (CurDistance+(RandomSamp*TempAADistance)));\r\nCurColor += DecodeSceneColorForMaterialNode(NewUV);\r\n//CurDistance+=(StepSize+(TempAADistance))/RadialSteps;\r\nSubstep++;\r\n}\r\nCurDistance+=StepSize;\r\nSubstep+=RadialOffset;\r\ni++;\r\n}\r\nCurColor = CurColor / ((int)DistanceSteps*(int)RadialSteps);\r\nreturn CurColor;\r\n}\n*** Original source before expression upgrade ***\nfloat3 CurColor;\r\nfloat2 BaseUV = MaterialFloat2(ScreenAlignedPosition(Parameters.ScreenPosition).xy);\r\nfloat2 NewUV = BaseUV;\r\nfloat StepSize = Distance / (int) DistanceSteps;\r\nfloat CurDistance=0;\r\nfloat2 CurOffset;\r\nfloat TwoPi = 6.283185;\r\nfloat Substep;\r\nfloat2 ScenePixels=Frame.RenderTargetSize*BaseUV;\r\nScenePixels+=Frame.TemporalAAParams.r;\r\nfloat2 RandomSamp = ((uint)(ScenePixels.x) + 2 * (uint)(ScenePixels.y)) % 5;\r\nRandomSamp+=Texture2DSample(Tex,TexSampler,ScenePixels);\r\nRandomSamp/=5;\r\nRandomSamp-=0.5;\r\nTempAARotation*=RandomSamp;\r\nTempAADistance*=StepSize*RandomSamp;\r\n\r\n\r\nint i=0;\r\nif (DistanceSteps<1)\r\n{\r\nreturn DecodeSceneColorForMaterialNode(NewUV);\r\n}\r\nelse\r\n{\r\n//CurDistance += 0.5*StepSize;\r\nwhile ( i < (int) DistanceSteps)\r\n{\r\n\r\n//CurDistance+=StepSize;\r\nfor (int j = 0; j < (int) RadialSteps; j++)\r\n{\r\nCurOffset.x = cos(TwoPi*((TempAARotation+Substep) / RadialSteps));\r\nCurOffset.y = sin(TwoPi*((TempAARotation+Substep) / RadialSteps));\r\nCurOffset *=DistanceMask;\r\nNewUV.x = BaseUV.x + (CurOffset.x * (CurDistance+(RandomSamp*TempAADistance)));\r\nNewUV.y = BaseUV.y + (CurOffset.y * (CurDistance+(RandomSamp*TempAADistance)));\r\nCurColor += DecodeSceneColorForMaterialNode(NewUV);\r\n//CurDistance+=(StepSize+(TempAADistance))/RadialSteps;\r\nSubstep++;\r\n}\r\nCurDistance+=StepSize;\r\nSubstep+=RadialOffset;\r\ni++;\r\n}\r\nCurColor = CurColor / ((int)DistanceSteps*(int)RadialSteps);\r\nreturn CurColor;\r\n}"
   outputType=CMOT_Float1
   showCode=true
   additionalOutputs [{outputName=Mask, outputType=CMOT_Float1}]
   inputs [
    {inputName=Distance, input {expression=&self:MaterialExpressionReroute_2}}
    {
     inputName=DistanceSteps
     input {expression=&self:MaterialExpressionReroute_3}
    }
    {
     inputName=RadialSteps
     input {expression=&self:MaterialExpressionReroute_4}
    }
    {
     inputName=TempAARotation
     input {expression=&self:MaterialExpressionReroute_5}
    }
    {
     inputName=DistanceMask
     input {expression=&self:MaterialExpressionReroute_6}
    }
    {inputName=Tex, input {expression=&self:MaterialExpressionReroute_7}}
    {
     inputName=TempAADistance
     input {expression=&self:MaterialExpressionReroute_8}
    }
    {
     inputName=RadialOffset
     input {expression=&self:MaterialExpressionReroute_9}
    }
    {inputName=DivideF, input {expression=&self:MaterialExpressionReroute_10}}
    {inputName=PDepth, input {expression=&self:MaterialExpressionReroute_11}}
    {
     inputName=PixelDepthOffset
     input {expression=&self:MaterialExpressionReroute_12}
    }
    {
     inputName=StencilIndexIn
     input {expression=&self:MaterialExpressionFunctionInput_13}
    }
   ]
  }
  MaterialExpressionReroute_13 = Reroute(None=@n5)
  MaterialExpressionReroute_14 = Reroute(None=@n6)
  MaterialExpressionReroute_15 = Reroute(None=@n7)
  MaterialExpressionReroute_16 = Reroute(None=@n8)
  MaterialExpressionReroute_17 = Reroute(None=@n9)
  MaterialExpressionReroute_18 = Reroute(None=@n10)
  MaterialExpressionReroute_19 = Reroute(None=@n11)
  MaterialExpressionReroute_20 = Reroute(None=@n12)
  MaterialExpressionReroute_21 = Reroute(None=@n13)
  MaterialExpressionReroute_22 = Reroute(None=@n23)
  MaterialExpressionReroute_23 = Reroute(None=@n24)
  n26 = Custom!(
   Distance=@MaterialExpressionReroute_13
   DistanceSteps=@MaterialExpressionReroute_14
   RadialSteps=@MaterialExpressionReroute_15
   TempAARotation=@MaterialExpressionReroute_16
   DistanceMask=@MaterialExpressionReroute_17
   Tex=@MaterialExpressionReroute_18
   TempAADistance=@MaterialExpressionReroute_19
   RadialOffset=@MaterialExpressionReroute_20
   DivideF=@MaterialExpressionReroute_21
   PDepth=@MaterialExpressionReroute_22
   PixelDepthOffset=@MaterialExpressionReroute_23
  ) {
   code="float CurColor = 0;\r\nfloat2 BaseUV = MaterialFloat2(ScreenAlignedPosition(Parameters.ScreenPosition).xy);\r\nfloat2 NewUV = BaseUV;\r\nfloat StepSize = Distance / (int) DistanceSteps;\r\nfloat CurDistance = 0;\r\nfloat2 CurOffset = 0;\r\nfloat TwoPi = 6.283185;\r\nfloat Substep = 0;\r\nfloat2 ScenePixels=View.BufferSizeAndInvSize.xy*BaseUV;\r\nScenePixels+=View.TemporalAAParams.r;\r\nfloat2 RandomSamp = ((uint)(ScenePixels.x) + 2 * (uint)(ScenePixels.y)) % 5;\r\nRandomSamp+=Texture2DSample(Tex,TexSampler,ScenePixels);\r\nRandomSamp/=5;\r\nRandomSamp-=0.5;\r\nTempAARotation*=RandomSamp;\r\nTempAADistance*=StepSize*RandomSamp;\r\n\r\n\r\nint i=0;\r\nif (DistanceSteps<1)\r\n{\r\n    Mask = 0;\r\n    return 0;\r\n}\r\nelse\r\n{\r\n    //CurDistance += 0.5*StepSize;\r\n    while ( i < (int) DistanceSteps)\r\n    {\r\n\r\n        //CurDistance+=StepSize;\r\n        for (int j = 0; j < (int) RadialSteps; j++)\r\n        {\r\n            CurOffset.x = cos(TwoPi*((TempAARotation+Substep) / RadialSteps));\r\n            CurOffset.y = sin(TwoPi*((TempAARotation+Substep) / RadialSteps));\r\n            CurOffset *=DistanceMask;\r\n            NewUV.x = BaseUV.x + (CurOffset.x * (CurDistance+(RandomSamp*TempAADistance)));\r\n            NewUV.y = BaseUV.y + (CurOffset.y * (CurDistance+(RandomSamp*TempAADistance)));\r\n            float tmpdepth = (SceneTextureLookupCommon(Parameters, NewUV, 13, false)).r;\r\n            float tmpscenedepth = CalcSceneDepth(NewUV);\r\n            float depthcomparemask = ceil(saturate(tmpscenedepth - tmpdepth+0.01))*ceil(saturate(DivideF-(tmpscenedepth-PDepth)+0.01));\r\n            tmpdepth = tmpdepth-PDepth+PixelDepthOffset;\r\n            CurColor += (min(1-saturate(max(0,(tmpdepth))/(DivideF+PixelDepthOffset))-0.99,0)+0.99)*saturate(sign(tmpdepth))*depthcomparemask;\r\n            //CurDistance+=(StepSize+(TempAADistance))/RadialSteps;\r\n            Substep++;\r\n        }\r\n        CurDistance+=StepSize;\r\n        Substep+=RadialOffset;\r\n        i++;\r\n    }\r\n    float tmpdepth2 = (SceneTextureLookupCommon(Parameters, BaseUV, 13, false)).r-PDepth+PixelDepthOffset;\r\n    float tmpscenedepth2 = CalcSceneDepth(BaseUV);\r\n    float depthcomparemask2 = ceil(saturate(tmpscenedepth2 - tmpdepth2+0.01))*ceil(saturate(DivideF-(tmpscenedepth2-PDepth)+0.01));\r\n    float mask = (min(1-saturate(max(0,(tmpdepth2))/(DivideF+PixelDepthOffset))-0.99,0)+0.99)*saturate(sign(tmpdepth2))*depthcomparemask2;\r\n    Mask = mask;\r\n    CurColor = CurColor / ((int)DistanceSteps*(int)RadialSteps);\r\n    return CurColor;\r\n}"
   desc="\n*** Original source before expression upgrade ***\nfloat3 CurColor;\r\nfloat2 BaseUV = MaterialFloat2(ScreenAlignedPosition(Parameters.ScreenPosition).xy);\r\nfloat2 NewUV = BaseUV;\r\nfloat StepSize = Distance / (int) DistanceSteps;\r\nfloat CurDistance=0;\r\nfloat2 CurOffset;\r\nfloat TwoPi = 6.283185;\r\nfloat Substep;\r\nfloat2 ScenePixels=View.RenderTargetSize*BaseUV;\r\nScenePixels+=View.TemporalAAParams.r;\r\nfloat2 RandomSamp = ((uint)(ScenePixels.x) + 2 * (uint)(ScenePixels.y)) % 5;\r\nRandomSamp+=Texture2DSample(Tex,TexSampler,ScenePixels);\r\nRandomSamp/=5;\r\nRandomSamp-=0.5;\r\nTempAARotation*=RandomSamp;\r\nTempAADistance*=StepSize*RandomSamp;\r\n\r\n\r\nint i=0;\r\nif (DistanceSteps<1)\r\n{\r\nreturn DecodeSceneColorForMaterialNode(NewUV);\r\n}\r\nelse\r\n{\r\n//CurDistance += 0.5*StepSize;\r\nwhile ( i < (int) DistanceSteps)\r\n{\r\n\r\n//CurDistance+=StepSize;\r\nfor (int j = 0; j < (int) RadialSteps; j++)\r\n{\r\nCurOffset.x = cos(TwoPi*((TempAARotation+Substep) / RadialSteps));\r\nCurOffset.y = sin(TwoPi*((TempAARotation+Substep) / RadialSteps));\r\nCurOffset *=DistanceMask;\r\nNewUV.x = BaseUV.x + (CurOffset.x * (CurDistance+(RandomSamp*TempAADistance)));\r\nNewUV.y = BaseUV.y + (CurOffset.y * (CurDistance+(RandomSamp*TempAADistance)));\r\nCurColor += DecodeSceneColorForMaterialNode(NewUV);\r\n//CurDistance+=(StepSize+(TempAADistance))/RadialSteps;\r\nSubstep++;\r\n}\r\nCurDistance+=StepSize;\r\nSubstep+=RadialOffset;\r\ni++;\r\n}\r\nCurColor = CurColor / ((int)DistanceSteps*(int)RadialSteps);\r\nreturn CurColor;\r\n}\n*** Original source before expression upgrade ***\nfloat3 CurColor;\r\nfloat2 BaseUV = MaterialFloat2(ScreenAlignedPosition(Parameters.ScreenPosition).xy);\r\nfloat2 NewUV = BaseUV;\r\nfloat StepSize = Distance / (int) DistanceSteps;\r\nfloat CurDistance=0;\r\nfloat2 CurOffset;\r\nfloat TwoPi = 6.283185;\r\nfloat Substep;\r\nfloat2 ScenePixels=Frame.RenderTargetSize*BaseUV;\r\nScenePixels+=Frame.TemporalAAParams.r;\r\nfloat2 RandomSamp = ((uint)(ScenePixels.x) + 2 * (uint)(ScenePixels.y)) % 5;\r\nRandomSamp+=Texture2DSample(Tex,TexSampler,ScenePixels);\r\nRandomSamp/=5;\r\nRandomSamp-=0.5;\r\nTempAARotation*=RandomSamp;\r\nTempAADistance*=StepSize*RandomSamp;\r\n\r\n\r\nint i=0;\r\nif (DistanceSteps<1)\r\n{\r\nreturn DecodeSceneColorForMaterialNode(NewUV);\r\n}\r\nelse\r\n{\r\n//CurDistance += 0.5*StepSize;\r\nwhile ( i < (int) DistanceSteps)\r\n{\r\n\r\n//CurDistance+=StepSize;\r\nfor (int j = 0; j < (int) RadialSteps; j++)\r\n{\r\nCurOffset.x = cos(TwoPi*((TempAARotation+Substep) / RadialSteps));\r\nCurOffset.y = sin(TwoPi*((TempAARotation+Substep) / RadialSteps));\r\nCurOffset *=DistanceMask;\r\nNewUV.x = BaseUV.x + (CurOffset.x * (CurDistance+(RandomSamp*TempAADistance)));\r\nNewUV.y = BaseUV.y + (CurOffset.y * (CurDistance+(RandomSamp*TempAADistance)));\r\nCurColor += DecodeSceneColorForMaterialNode(NewUV);\r\n//CurDistance+=(StepSize+(TempAADistance))/RadialSteps;\r\nSubstep++;\r\n}\r\nCurDistance+=StepSize;\r\nSubstep+=RadialOffset;\r\ni++;\r\n}\r\nCurColor = CurColor / ((int)DistanceSteps*(int)RadialSteps);\r\nreturn CurColor;\r\n}"
   outputType=CMOT_Float1
   showCode=true
   additionalOutputs [{outputName=Mask, outputType=CMOT_Float1}]
   inputs [
    {inputName=Distance, input {expression=&self:MaterialExpressionReroute_13}}
    {
     inputName=DistanceSteps
     input {expression=&self:MaterialExpressionReroute_14}
    }
    {
     inputName=RadialSteps
     input {expression=&self:MaterialExpressionReroute_15}
    }
    {
     inputName=TempAARotation
     input {expression=&self:MaterialExpressionReroute_16}
    }
    {
     inputName=DistanceMask
     input {expression=&self:MaterialExpressionReroute_17}
    }
    {inputName=Tex, input {expression=&self:MaterialExpressionReroute_18}}
    {
     inputName=TempAADistance
     input {expression=&self:MaterialExpressionReroute_19}
    }
    {
     inputName=RadialOffset
     input {expression=&self:MaterialExpressionReroute_20}
    }
    {inputName=DivideF, input {expression=&self:MaterialExpressionReroute_21}}
    {inputName=PDepth, input {expression=&self:MaterialExpressionReroute_22}}
    {
     inputName=PixelDepthOffset
     input {expression=&self:MaterialExpressionReroute_23}
    }
   ]
  }
  n27 = Custom(
   Distance=@MaterialExpressionReroute_2
   DistanceSteps=@MaterialExpressionReroute_3
   RadialSteps=@MaterialExpressionReroute_4
   TempAARotation=@MaterialExpressionReroute_5
   DistanceMask=@MaterialExpressionReroute_6
   Tex=@MaterialExpressionReroute_7
   TempAADistance=@MaterialExpressionReroute_8
   RadialOffset=@MaterialExpressionReroute_9
   DivideF=@MaterialExpressionReroute_10
   PDepth=@MaterialExpressionReroute_11
   PixelDepthOffset=@MaterialExpressionReroute_12
  ) {
   code="float CurColor = 0;\r\nfloat2 BaseUV = MaterialFloat2(ScreenAlignedPosition(Parameters.ScreenPosition).xy);\r\nfloat2 NewUV = BaseUV;\r\nfloat StepSize = Distance / (int) DistanceSteps;\r\nfloat CurDistance = 0;\r\nfloat2 CurOffset = 0;\r\nfloat TwoPi = 6.283185;\r\nfloat Substep = 0;\r\nfloat2 ScenePixels=View.BufferSizeAndInvSize.xy*BaseUV;\r\nScenePixels+=View.TemporalAAParams.r;\r\nfloat2 RandomSamp = ((uint)(ScenePixels.x) + 2 * (uint)(ScenePixels.y)) % 5;\r\nRandomSamp+=Texture2DSample(Tex,TexSampler,ScenePixels);\r\nRandomSamp/=5;\r\nRandomSamp-=0.5;\r\nTempAARotation*=RandomSamp;\r\nTempAADistance*=StepSize*RandomSamp;\r\n\r\n\r\nint i=0;\r\nif (DistanceSteps<1)\r\n{\r\n    Mask = 0;\r\n    return 0;\r\n}\r\nelse\r\n{\r\n    //CurDistance += 0.5*StepSize;\r\n    while ( i < (int) DistanceSteps)\r\n    {\r\n        //CurDistance+=StepSize;\r\n        for (int j = 0; j < (int) RadialSteps; j++)\r\n        {\r\n            CurOffset.x = cos(TwoPi*((TempAARotation+Substep) / RadialSteps));\r\n            CurOffset.y = sin(TwoPi*((TempAARotation+Substep) / RadialSteps));\r\n            CurOffset *=DistanceMask;\r\n            NewUV.x = BaseUV.x + (CurOffset.x * (CurDistance+(RandomSamp*TempAADistance)));\r\n            NewUV.y = BaseUV.y + (CurOffset.y * (CurDistance+(RandomSamp*TempAADistance)));\r\n            float tmpdepth = CalcSceneDepth(NewUV)-PDepth+PixelDepthOffset;\r\n            CurColor += (min(1-saturate(max(0,(tmpdepth))/(DivideF+PixelDepthOffset))-0.99,0)+0.99)*saturate(sign(tmpdepth));\r\n            //CurDistance+=(StepSize+(TempAADistance))/RadialSteps;\r\n            Substep++;\r\n        }\r\n        CurDistance+=StepSize;\r\n        Substep+=RadialOffset;\r\n        i++;\r\n    }\r\n    float tmpdepth2 = CalcSceneDepth(BaseUV)-PDepth+PixelDepthOffset;\r\n    float mask = (min(1-saturate(max(0,(tmpdepth2))/(DivideF+PixelDepthOffset))-0.99,0)+0.99)*saturate(sign(tmpdepth2));\r\n    Mask = mask;\r\n    CurColor = CurColor / ((int)DistanceSteps*(int)RadialSteps);\r\n    return CurColor;\r\n}"
   desc="\n*** Original source before expression upgrade ***\nfloat3 CurColor;\r\nfloat2 BaseUV = MaterialFloat2(ScreenAlignedPosition(Parameters.ScreenPosition).xy);\r\nfloat2 NewUV = BaseUV;\r\nfloat StepSize = Distance / (int) DistanceSteps;\r\nfloat CurDistance=0;\r\nfloat2 CurOffset;\r\nfloat TwoPi = 6.283185;\r\nfloat Substep;\r\nfloat2 ScenePixels=View.RenderTargetSize*BaseUV;\r\nScenePixels+=View.TemporalAAParams.r;\r\nfloat2 RandomSamp = ((uint)(ScenePixels.x) + 2 * (uint)(ScenePixels.y)) % 5;\r\nRandomSamp+=Texture2DSample(Tex,TexSampler,ScenePixels);\r\nRandomSamp/=5;\r\nRandomSamp-=0.5;\r\nTempAARotation*=RandomSamp;\r\nTempAADistance*=StepSize*RandomSamp;\r\n\r\n\r\nint i=0;\r\nif (DistanceSteps<1)\r\n{\r\nreturn DecodeSceneColorForMaterialNode(NewUV);\r\n}\r\nelse\r\n{\r\n//CurDistance += 0.5*StepSize;\r\nwhile ( i < (int) DistanceSteps)\r\n{\r\n\r\n//CurDistance+=StepSize;\r\nfor (int j = 0; j < (int) RadialSteps; j++)\r\n{\r\nCurOffset.x = cos(TwoPi*((TempAARotation+Substep) / RadialSteps));\r\nCurOffset.y = sin(TwoPi*((TempAARotation+Substep) / RadialSteps));\r\nCurOffset *=DistanceMask;\r\nNewUV.x = BaseUV.x + (CurOffset.x * (CurDistance+(RandomSamp*TempAADistance)));\r\nNewUV.y = BaseUV.y + (CurOffset.y * (CurDistance+(RandomSamp*TempAADistance)));\r\nCurColor += DecodeSceneColorForMaterialNode(NewUV);\r\n//CurDistance+=(StepSize+(TempAADistance))/RadialSteps;\r\nSubstep++;\r\n}\r\nCurDistance+=StepSize;\r\nSubstep+=RadialOffset;\r\ni++;\r\n}\r\nCurColor = CurColor / ((int)DistanceSteps*(int)RadialSteps);\r\nreturn CurColor;\r\n}\n*** Original source before expression upgrade ***\nfloat3 CurColor;\r\nfloat2 BaseUV = MaterialFloat2(ScreenAlignedPosition(Parameters.ScreenPosition).xy);\r\nfloat2 NewUV = BaseUV;\r\nfloat StepSize = Distance / (int) DistanceSteps;\r\nfloat CurDistance=0;\r\nfloat2 CurOffset;\r\nfloat TwoPi = 6.283185;\r\nfloat Substep;\r\nfloat2 ScenePixels=Frame.RenderTargetSize*BaseUV;\r\nScenePixels+=Frame.TemporalAAParams.r;\r\nfloat2 RandomSamp = ((uint)(ScenePixels.x) + 2 * (uint)(ScenePixels.y)) % 5;\r\nRandomSamp+=Texture2DSample(Tex,TexSampler,ScenePixels);\r\nRandomSamp/=5;\r\nRandomSamp-=0.5;\r\nTempAARotation*=RandomSamp;\r\nTempAADistance*=StepSize*RandomSamp;\r\n\r\n\r\nint i=0;\r\nif (DistanceSteps<1)\r\n{\r\nreturn DecodeSceneColorForMaterialNode(NewUV);\r\n}\r\nelse\r\n{\r\n//CurDistance += 0.5*StepSize;\r\nwhile ( i < (int) DistanceSteps)\r\n{\r\n\r\n//CurDistance+=StepSize;\r\nfor (int j = 0; j < (int) RadialSteps; j++)\r\n{\r\nCurOffset.x = cos(TwoPi*((TempAARotation+Substep) / RadialSteps));\r\nCurOffset.y = sin(TwoPi*((TempAARotation+Substep) / RadialSteps));\r\nCurOffset *=DistanceMask;\r\nNewUV.x = BaseUV.x + (CurOffset.x * (CurDistance+(RandomSamp*TempAADistance)));\r\nNewUV.y = BaseUV.y + (CurOffset.y * (CurDistance+(RandomSamp*TempAADistance)));\r\nCurColor += DecodeSceneColorForMaterialNode(NewUV);\r\n//CurDistance+=(StepSize+(TempAADistance))/RadialSteps;\r\nSubstep++;\r\n}\r\nCurDistance+=StepSize;\r\nSubstep+=RadialOffset;\r\ni++;\r\n}\r\nCurColor = CurColor / ((int)DistanceSteps*(int)RadialSteps);\r\nreturn CurColor;\r\n}"
   outputType=CMOT_Float1
   showCode=true
   additionalOutputs [{outputName=Mask, outputType=CMOT_Float1}]
   inputs [
    {inputName=Distance, input {expression=&self:MaterialExpressionReroute_2}}
    {
     inputName=DistanceSteps
     input {expression=&self:MaterialExpressionReroute_3}
    }
    {
     inputName=RadialSteps
     input {expression=&self:MaterialExpressionReroute_4}
    }
    {
     inputName=TempAARotation
     input {expression=&self:MaterialExpressionReroute_5}
    }
    {
     inputName=DistanceMask
     input {expression=&self:MaterialExpressionReroute_6}
    }
    {inputName=Tex, input {expression=&self:MaterialExpressionReroute_7}}
    {
     inputName=TempAADistance
     input {expression=&self:MaterialExpressionReroute_8}
    }
    {
     inputName=RadialOffset
     input {expression=&self:MaterialExpressionReroute_9}
    }
    {inputName=DivideF, input {expression=&self:MaterialExpressionReroute_10}}
    {inputName=PDepth, input {expression=&self:MaterialExpressionReroute_11}}
    {
     inputName=PixelDepthOffset
     input {expression=&self:MaterialExpressionReroute_12}
    }
   ]
  }
  n28 = StaticSwitch(True=@n25.return, False=StaticSwitch(True=@n26.return, False=@n27.return, Value=@n18), Value=@n22)
  n29 = Saturate(None=@n28)
  n30 = StaticSwitch(True=@n25.Mask, False=StaticSwitch(True=@n26.Mask, False=@n27.Mask, Value=@n18), Value=@n22)
  n31 = Saturate(None=@n30)
  n32 = FunctionInput() {bUsePreviewValueAsDefault=true, inputName=OutlineStrength, inputType=FunctionInput_Scalar, previewValue=<1 0 0 1>}
  n33 = Constant() {r=0.6}
  n34 = Saturate(None=Subtract(A=@n28, B=Multiply(A=@n30, B=@n33)))
  n35 = FunctionInput() {bUsePreviewValueAsDefault=true, inputName=InnerGlowStrength, inputType=FunctionInput_Scalar, sortPriority=50, previewValue=<1 0 0 1>}
  n36 = FunctionInput() {bUsePreviewValueAsDefault=true, inputName=InnerGlowPower, inputType=FunctionInput_Scalar, sortPriority=40, previewValue=<7 0 0 1>}
  n37 = Power(Base=OneMinus(None=@n34), Exp=@n36)
  n38 = FunctionInput() {bUsePreviewValueAsDefault=true, inputName=InnerGlowMaxArea, inputType=FunctionInput_Scalar, sortPriority=30, previewValue=<10 0 0 1>}
  n39 = Multiply(A=Power(Base=@n30, Exp=Constant() {r=2.5}), B=@n38) {constB=1000}
  n40 = LinearInterpolate(
   A=Multiply(A=Subtract(A=@n29, B=@n31), B=@n32)
   B=LinearInterpolate(A=Multiply(A=@n34, B=@n32), B=Multiply(A=@n35, B=@n37), Alpha=Saturate(None=@n39)) {constB=0}
   Alpha=FunctionInput() {bUsePreviewValueAsDefault=true, inputName=------EnableInnerGlow------, inputType=FunctionInput_Scalar, sortPriority=20}
  )
  FunctionOutput(None=Add(A=Multiply(A=@MaterialExpressionFunctionInput_13) {constB=0}, B=Saturate(None=@n40)))
  FunctionOutput(None=@n30) {outputName=BlurAreaDepth0-1, sortPriority=1}
 ]
}
