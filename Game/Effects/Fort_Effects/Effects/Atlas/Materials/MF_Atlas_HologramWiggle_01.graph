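# VFX material function that produces an animated hologram wiggle offset, apparently for World Position Offset, plus a 0-1 intensity mask.
# Two copies of a 64x64 noise texture, sampled as normals, pan across screen UVs, each scaled by its own vector intensity input.
# Two sines of differing periods multiply, take a Bias, then are clamped and sharpened by a power into bursts.
# That mask gates only layer B before it adds to layer A.
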MaterialFunction /Game/Effects/Fort_Effects/Effects/Atlas/Materials/MF_Atlas_HologramWiggle_01
graph {
 properties {bExposeToLibrary=true, libraryCategoriesText [VFX]}
 nodes [
  n1 = ScreenPosition()
  n2 = FunctionInput(Preview=Time()) {bUsePreviewValueAsDefault=true, inputName=Time, inputType=FunctionInput_Scalar, sortPriority=2}
  n3 = TextureSample(UVs=Panner(Coordinate=@n1.ViewportUV, Time=@n2) {bFractionalPart=true, speedX=1, speedY=0.22}) {samplerType=SAMPLERTYPE_Normal, texture=/Game/Global/Textures/Default/Noise/64x64Noise}
  n4 = FunctionInput(Preview=VectorParameter() {parameterName=WPO Intensity A, defaultValue=<rgba 1.0055 0.9055 1.1172 0>}.RGB) {bUsePreviewValueAsDefault=true, inputName=WPO Intensity A}
  n5 = Multiply(
   A=TextureSample(UVs=Panner(Coordinate=@n1.ViewportUV, Time=@n2) {bFractionalPart=true, speedX=-0.61, speedY=2}) {samplerType=SAMPLERTYPE_Normal, texture=&64x64Noise}.RGB
   B=FunctionInput(Preview=VectorParameter() {parameterName=WPO Intensity B, defaultValue=<rgba 0 4 0 0>}.RGB) {bUsePreviewValueAsDefault=true, inputName=WPO Intensity B, sortPriority=1}
  )
  n6 = Multiply(A=ConstantBiasScale(None=Sine(None=@n2) {period=1.1225}), B=ConstantBiasScale(None=Sine(None=@n2) {period=3.33}) {scale=0.65})
  n7 = FunctionInput() {bUsePreviewValueAsDefault=true, inputName=Bias, inputType=FunctionInput_Scalar, sortPriority=3}
  n8 = Power(Base=Clamp(None=Add(A=@n6, B=@n7))) {constExponent=3.5}
  FunctionOutput(None=Add(A=Multiply(A=@n3.RGB, B=@n4), B=Multiply(A=@n5, B=@n8)))
  FunctionOutput(None=@n8) {outputName=Intense 0-1, sortPriority=1}
 ]
}
