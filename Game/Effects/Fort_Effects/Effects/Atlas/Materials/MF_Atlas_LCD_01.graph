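# MF_Atlas_LCD_01 is a library-exposed VFX material function that generates a procedural LCD pattern for holograms and tech effects.
# It tiles UVs by separate U and V repeat inputs, then multiplies the fractional coordinate by its inverse, forming texture-free cells that peak at center.
# A dot product with one sums both axes into an alpha, peaking at 0.5, that blends the dark and light inputs, whose previews default to named parameters.
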MaterialFunction /Game/Effects/Fort_Effects/Effects/Atlas/Materials/MF_Atlas_LCD_01
graph {
 properties {bExposeToLibrary=true, description="LCD Pattern for holograms, tech, etc", libraryCategoriesText [VFX]}
 nodes [
  n1 = ScalarParameter() {defaultValue=0.5, parameterName=LCD Darks}
  n2 = ScalarParameter() {defaultValue=2, parameterName=LCD Lights}
  n3 = FunctionInput(Preview=ScalarParameter() {defaultValue=48, parameterName=LCD U}) {bUsePreviewValueAsDefault=true, inputName=U Repeat, inputType=FunctionInput_Scalar}
  n4 = FunctionInput(Preview=ScalarParameter() {defaultValue=256, parameterName=LCD V}) {bUsePreviewValueAsDefault=true, inputName=V Repeat, inputType=FunctionInput_Scalar, sortPriority=1}
  n5 = Frac(None=Multiply(A=TextureCoordinate(), B=AppendVector(A=@n3, B=@n4)))
  n6 = Multiply(A=OneMinus(None=@n5), B=@n5)
  n7 = Constant() {r=1}
  FunctionOutput(None=LinearInterpolate(
    A=FunctionInput(Preview=@n1) {bUsePreviewValueAsDefault=true, inputName=Dark Value, inputType=FunctionInput_Scalar, sortPriority=2}
    B=FunctionInput(Preview=@n2) {bUsePreviewValueAsDefault=true, inputName=Light Value, inputType=FunctionInput_Scalar, sortPriority=3}
    Alpha=DotProduct(A=@n6, B=@n7)
   ))
 ]
}
